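# Two-sided cloth base material for the ShovelCropUniversal pickaxe, layering animated electrical effects over the 2023 anime character shader.
# Noise-driven lightning streaks, screen-stabilized stardust and clouds, thin-film color, and sphere-masked fresnel rims provide the effect layers.
# Emissive combines lit color with glow layers, adding sheen and outline on front faces only.
# Sphere-masked World Position Offset combines time-stepped noise jolts, a sine wave and normal flare, with optional texture masking.
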Material /BRCosmetics/Weapons/FORT_Melee/Pickaxe_ShovelCropUniversal/Materials/M_ShovelCrop_Base
graph {
 properties {
  bAutomaticallySetUsageInEditor=false
  bUseMaterialAttributes=true
  bUsedWithClothing=true
  bUsedWithMeshDeformer=true
  bUsedWithMorphTargets=true
  bUsedWithNiagaraMeshParticles=true
  bUsedWithSkeletalMesh=true
  bUsedWithStaticLighting=true
  pixelDepthOffsetMode=PDOM_Legacy
  previewMesh=/BRCosmetics/Weapons/FORT_Melee/Pickaxe_ShovelCrop/Meshes/ShovelCrop_Axe
  shadingModel=MSM_Cloth
  twoSided=true
  usedShadingModels=Cloth
 }
 nodes [
  n1 = PreSkinnedNormal()
  Abs(None=StaticComponentMaskParameter(None=VertexInterpolator(VS=@n1).PS) {defaultB=true, parameterName=FresnelAxis})
  MaterialFunctionCall() {materialFunction=/BRCosmetics/Animation/Props/Gadgets_AnglePatch/Materials/MF_PreskinnedWorldPosition_AnglePatch}
  n2 = Multiply(A=NamedRerouteUsage(), B=ScalarParameter() {defaultValue=1, group=GridEffect, parameterName=LightingNoise_Tiling})
  n3 = ChannelMaskParameter(None=@n2) {maskChannel=Green, parameterName=Lightening Noise Dir, defaultValue=<rgba 0 1 0 0>}
  n4 = ScalarParameter() {defaultValue=1, group=GridEffect, parameterName=LightingNoise_Frequency}
  n5 = Time() {period=1}
  n6 = Add(A=Multiply(A=@n2) {constB=2.3}, B=AppendVector(A=Constant(), B=Multiply(A=@n5) {constB=0.61}))
  n7 = ScalarParameter() {defaultValue=2, group=GridEffect, parameterName=LightingNoise_Snap}
  n8 = Time() {period=1}
  n9 = Add(A=Multiply(A=@n2, B=OneMinus(None=@n7)) {constB=-2}, B=AppendVector(A=Constant(), B=Multiply(A=@n8) {constB=0.37}))
  n10 = Add(
   A=ChannelMaskParameter(None=TextureSampleParameter2D(UVs=@n6) {parameterName=Lightening Noise, texture=/Game/Global/Textures/Default/Noise/3xPerlin_M}.RGB) {parameterName=Lightening Noise Channel 1}.R
   B=ChannelMaskParameter(None=TextureSampleParameter2D(UVs=@n9) {parameterName=Lightening Noise, texture=&3xPerlin_M}.RGB) {parameterName=Lightening Noise Channel 2}.R
  )
  n11 = ScalarParameter() {defaultValue=1, group=GridEffect, parameterName=LightingNoise_Distortion}
  n12 = Multiply(
   A=Add(A=Multiply(A=@n3.R, B=@n4), B=Multiply(A=@n10, B=@n11))
   B=MaterialFunctionCall(Multiplier=Constant() {r=2}) {materialFunction=&Pi}.Result
  )
  n13 = Multiply(A=Time() {period=1}, B=ScalarParameter() {defaultValue=1, group=GridEffect, parameterName=LightingNoise_PanSpeed})
  n14 = Multiply(
   A=Add(A=Add(A=@n12, B=@n13), B=Time() {period=1}) {constA=1}
   B=ScalarParameter() {defaultValue=0.5, group=GridEffect, parameterName=LightingNoise_FlowSpeed}
  )
  n15 = Power(
   Base=ConstantBiasScale(None=Sine(None=@n14))
   Exp=ScalarParameter() {defaultValue=0.7, group=GridEffect, parameterName=LightingNoise_Sharpness}
  )
  n16 = Multiply(
   A=ChannelMaskParameter(None=@n2) {maskChannel=Green, parameterName=Lightening Noise Dir 2, defaultValue=<rgba 0 1 0 0>}.R
   B=ScalarParameter() {defaultValue=1, group=GridEffect, parameterName=LightingNoise_Frequency_2}
  )
  n17 = Multiply(A=ScalarParameter() {defaultValue=1, group=GridEffect, parameterName=LightingNoise_Distortion_2}, B=@n10)
  n18 = Constant() {r=2}
  n19 = Add(
   A=Multiply(
    A=Add(A=@n16, B=@n17)
    B=MaterialFunctionCall(Multiplier=@n18) {materialFunction=&Pi}.Result
   )
   B=Multiply(A=Time() {period=1}, B=ScalarParameter() {defaultValue=1, group=GridEffect, parameterName=LightingNoise_PanSpeed_2})
  )
  n20 = Time() {period=1}
  n21 = Sine(None=Multiply(A=Add(A=@n19, B=@n20) {constA=1}, B=ScalarParameter() {defaultValue=0.5, group=GridEffect, parameterName=LightingNoise_FlowSpeed}))
  Max(
   A=@n15
   B=Power(Base=ConstantBiasScale(None=@n21), Exp=ScalarParameter() {defaultValue=0.7, group=GridEffect, parameterName=LightingNoise_Sharpness_2})
  )
  n22 = NamedRerouteDeclaration(None=MaterialFunctionCall!(ObjectPivotLocation, ObjectPositionOffset=VectorParameter() {parameterName=ScreenSpace Pos, defaultValue=<rgba 0 0 0 1>}.RGB) {materialFunction=/Game/Characters/Player/Common/MaterialFunctions/Superlevel/ImprovedStabilizedScreenSpaceUVs}.StabilizedUVs) {name=ScreenUV, nodeColor=<rgba 1 0 0.6797 1>}
  n23 = ScalarParameter() {defaultValue=4, parameterName=StardustScale}
  n24 = Constant() {r=3}
  n25 = Multiply(A=Time(), B=@n24) {constB=3}
  n26 = Floor(None=@n25)
  n27 = TextureSample(UVs=Add(A=Multiply(A=@n22, B=@n23) {constB=2048}, B=Divide(A=@n26, B=@n24) {constB=3})) {texture=/BRCosmetics/Weapons/FORT_Melee/Pickaxe_BullKeynote/Textures/T_BullKeyNote_Noise}
  n28 = Power(
   Base=MaterialFunctionCall(A=@n27.R, B=@n27.R, C=@n27.R, Alpha=Frac(None=@n25)) {materialFunction=&Lerp_3Color}.Result
   Exp=ScalarParameter() {defaultValue=1, parameterName=Stardust Exp}
  ) {constExponent=1}
  n29 = Multiply(A=NamedRerouteUsage(), B=ScalarParameter() {defaultValue=4, parameterName=CloudsScale})
  n30 = Panner(Coordinate=@n29) {speedY=0.076}
  n31 = AppendVector(A=@n29, B=Multiply(A=Time(), B=ScalarParameter() {defaultValue=0.3, parameterName=CloudsAnimSpeed}))
  n32 = Power(
   Base=ConstantBiasScale(None=Noise!(World Position=@n31, FilterWidth) {bTiling=true, noiseFunction=NOISEFUNCTION_GradientTex3D, repeatSize=4})
   Exp=ScalarParameter() {defaultValue=2, parameterName=CloudsExp}
  )
  n33 = Multiply(
   A=Multiply(
    A=TextureSampleParameter2D(UVs=@n30) {parameterName=Inner Lightening Noise, texture=/Game/Global/Textures/Default/Masks_and_Gradients/chunkma_2}.R
    B=Saturate(None=@n32)
   )
   B=ScalarParameter() {defaultValue=3, parameterName=CloudsBoost}
  ) {constA=3}
  NamedRerouteDeclaration(None=LinearInterpolate(A=ScalarParameter() {defaultValue=0.03, parameterName=MinBrightness}, Alpha=Add(A=@n28, B=@n33)) {constA=0.03}) {name=InnerElectricityEffect, nodeColor=<rgba 1 0 0.1172 1>}
  n34 = Transform(None=VectorParameter() {parameterName=Fresnel Axis, defaultValue=<rgba 0 0 0 1>}.RGB)
  n35 = CameraVectorWS()
  n36 = PixelNormalWS()
  n37 = VertexInterpolator(VS=LocalPosition() {localOrigin=InstancePreSkinning}.XYZ)
  n38 = VectorParameter() {parameterName=Fresnel_Mask_Pos, defaultValue=<rgba 0.4 0.15 0 1>}
  n39 = SphereMask(
   A=@n37.PS
   B=@n38.RGB
   Radius=ScalarParameter() {defaultValue=200, group=Lightening RimLight, parameterName=FresnelAxis_Mask_Radius}
   Hardness=ScalarParameter() {defaultValue=1, group=Lightening RimLight, parameterName=FresnelAxis_Mask_Hardness}
  )
  n40 = LinearInterpolate(
   A=OneMinus(None=MaterialFunctionCall(Axis=@n34) {materialFunction=&AxisAlignedFresnel}.Fresnel)
   B=OneMinus(None=DotProduct(A=@n35, B=@n36))
   Alpha=@n39
  )
  n41 = LinearInterpolate(
   A=LinearInterpolate(
    A=ScalarParameter() {defaultValue=1.5, group=Lightening RimLight, parameterName=Fresnel_Exponent, sortPriority=7}
    B=ScalarParameter() {defaultValue=1.5, group=Lightening RimLight, parameterName=FresnelAxisMask_Exponent, sortPriority=7}
    Alpha=@n39
   )
   B=ScalarParameter() {defaultValue=1.5, group=Lightening RimLight, parameterName=Fresnel_Exponent_Fingers, sortPriority=7}
   Alpha=SphereMask(
    A=@n37.PS
    B=@n38.RGB
    Radius=ScalarParameter() {defaultValue=200, group=Lightening RimLight, parameterName=Fresnel_Mask_Radius}
    Hardness=ScalarParameter() {defaultValue=1, group=Lightening RimLight, parameterName=Fresnel_Mask_Hardness}
   )
  )
  NamedRerouteDeclaration(None=SmoothStep(
    Min=ScalarParameter() {defaultValue=1, group=Lightening RimLight, parameterName=FresnelBrightness min, sliderMin=4, sortPriority=6}
    Max=ScalarParameter() {defaultValue=1, group=Lightening RimLight, parameterName=FresnelBrightness max, sliderMin=4, sortPriority=6}
    Value=Power(Base=@n40, Exp=@n41) {constExponent=1}
   ) {constMax=0.8, constMin=0.5, constValue=0.4}) {name=Outline Fresnel, nodeColor=<rgba 0.3594 1 0 1>}
  n42 = NamedRerouteUsage()
  n43 = ScalarParameter() {defaultValue=1, group=Lightening RimLight, parameterName=ThinFilmScale, sortPriority=5}
  n44 = Power(
   Base=NamedRerouteUsage()
   Exp=ScalarParameter() {defaultValue=0.5, group=06_Thin Film | Side Kick, parameterName=SideKick.ThinFilmWarp, sortPriority=4}
  )
  n45 = TextureSampleParameter2D(UVs=Add(A=Multiply(A=@n43, B=@n44), B=ScalarParameter() {group=06_Thin Film | Side Kick, parameterName=SideKick.ThinFilmOffset})) {
   group=06_Thin Film | Side Kick
   parameterName=SideKick.ThinFilmTexture
   samplerSource=SSM_Wrap_WorldGroupSettings
   sortPriority=-25
   texture=/Game/Global/Textures/Default/Thinfilm/T_ThinFilm_Spectrum_COLOR
  }
  n46 = Power(
   Base=NamedRerouteUsage()
   Exp=ScalarParameter() {defaultValue=2, group=06_Thin Film | Side Kick, parameterName=SideKick.ThinFilmExponent, sortPriority=3}
  )
  NamedRerouteDeclaration(None=LinearInterpolate(A=@n42, B=Add(A=@n42, B=@n45.RGB), Alpha=Saturate(None=@n46))) {name=Thin Film, nodeColor=<rgba 0.6172 1 0 1>}
  n47 = MaterialFunctionCall!(CustomUV) {materialFunction=/Game/Global/Materials/Character_Parent_2023/Anime_Parent_2023/MaterialFunctions/MFC_AnimeShader_2023_Pre}
  NamedRerouteDeclaration(None=@n47.Color_Lit) {name=Color Lit, nodeColor=<rgba 1 0 0.9375 1>}
  n48 = Power(Base=@n40, Exp=ScalarParameter() {defaultValue=1, group=Lightening RimLight, parameterName=Sheen_Exponent, sortPriority=11}) {constExponent=1}
  NamedRerouteDeclaration(None=SmoothStep(
    Min=ScalarParameter() {defaultValue=0.5, group=Lightening RimLight, parameterName=Sheen_MinRange, sortPriority=12}
    Max=ScalarParameter() {defaultValue=0.8, group=Lightening RimLight, parameterName=Sheen_MaxRange, sortPriority=13}
    Value=@n48
   ) {constMax=0.8, constMin=0.5, constValue=0.4}) {name=Sheen fresnel, nodeColor=<rgba 1 0 0.6797 1>}
  NamedRerouteDeclaration(None=LinearInterpolate(
    A=ScalarParameter() {defaultValue=0.1, group=GridEffect, parameterName=LightingNoise_Intensity_Min}
    B=ScalarParameter() {defaultValue=1, group=GridEffect, parameterName=LightingNoise_Intensity_Max}
    Alpha=@n15
   )) {name=LightingStreak, nodeColor=<rgba 0 1 1 1>}
  NamedRerouteDeclaration(None=SmoothStep(Value=@n48) {constMax=0.55, constMin=0.5}) {name=Sheen fresnel_Mask, nodeColor=<rgba 1 0 0.6797 1>}
  n49 = ScalarParameter() {defaultValue=-0.1, parameterName=EdgeMask bias}
  n50 = ScalarParameter() {defaultValue=1, parameterName=EdgeMask Scale}
  n51 = PixelNormalWS()
  n52 = Multiply(
   A=@n50
   B=DotProduct(
    A=Transform(None=@n51) {transformSourceType=TRANSFORMSOURCE_World, transformType=TRANSFORM_View}
    B=VectorParameter() {parameterName=Edge Mask Light Pos, defaultValue=<rgba 1 0 0 1>}.RGB
   )
  )
  NamedRerouteDeclaration(None=Saturate(None=Add(A=@n49, B=@n52))) {name=EdgeMask, nodeColor=<rgba 1 0.3047 0 1>}
  NamedRerouteDeclaration(None=Fresnel(ExponentIn=Constant() {r=1}) {baseReflectFraction=0}) {name=Fresnel, nodeColor=<rgba 1 0.2109 0 1>}
  NamedRerouteDeclaration(None=SphereMask(
    A=@n37.PS
    B=VectorParameter() {parameterName=Fresnel_Top_Mask_Pos, defaultValue=<rgba 0.4 0.15 0 1>}.RGB
    Radius=ScalarParameter() {defaultValue=200, group=Lightening RimLight, parameterName=Fresnel_Top_Mask_Radius}
    Hardness=ScalarParameter() {defaultValue=1, group=Lightening RimLight, parameterName=Fresnel_Top_Mask_Hardness}
   )) {name=TopMask, nodeColor=<rgba 0.5469 1 0 1>}
  n53 = ComponentMask(None=TransformPosition(None=WorldPosition().XYZ) {transformSourceType=TRANSFORMPOSSOURCE_World, transformType=TRANSFORMPOSSOURCE_View}) {r=true}
  Saturate(None=Add(A=Multiply(A=@n53, B=@n50), B=@n49))
  n54 = Subtract(A=ComponentMask(None=ScreenPosition().ViewportUV) {r=true}) {constB=0.5}
  Saturate(None=Add(A=Multiply(A=@n50, B=@n54)) {constB=0.5})
  n55 = Subtract(
   A=TransformPosition(None=Constant3Vector() {constant=<rgba 100 50 200 0>}) {transformSourceType=TRANSFORMPOSSOURCE_View, transformType=TRANSFORMPOSSOURCE_World}
   B=WorldPosition().XYZ
  )
  Saturate(None=DotProduct(A=Normalize(VectorInput=@n55), B=PixelNormalWS()))
  n56 = MaterialFunctionCall!(CustomLightVector, OverrideNormal_WS, STT=@n47.STT, Normal_Map=@n47.STT, DFL=@n47.STT, FX=@n47.STT) {materialFunction=/Game/Global/Materials/Character_Parent_2023/Anime_Parent_2023/MaterialFunctions/MFC_AnimeShader_2023_Calculations}
  n57 = Multiply(A=@n47.Color_Lit, B=ScalarParameter() {defaultValue=0.8, group=02_Anime Parameters, parameterName=Emissive_Intensity, sliderMax=10})
  n58 = Multiply(A=NamedRerouteUsage(), B=NamedRerouteUsage())
  n59 = NamedRerouteUsage()
  n60 = NamedRerouteUsage()
  n61 = VectorParameter() {group=01_Anime Maps, parameterName=Top Color, sortPriority=20, defaultValue=<rgba 1 0.939 0.0652 1>}
  n62 = ScalarParameter() {defaultValue=1, group=02_Anime Parameters, parameterName=Top_Intensity, sliderMax=10}
  n63 = Add(A=Add(A=Add(A=@n57, B=@n58), B=Multiply(A=@n59, B=@n60)), B=Multiply(A=NamedRerouteUsage(), B=Multiply(A=@n61.RGB, B=@n62)))
  n64 = VectorParameter() {group=01_Anime Maps, parameterName=Fresnel_sheen_Color, sortPriority=20, defaultValue=<rgba 1 0.9653 0.4688 1>}
  n65 = ScalarParameter() {defaultValue=1, group=02_Anime Parameters, parameterName=Fresnel Sheen Intensity, sliderMax=10}
  n66 = ComponentMask(None=Transform(None=PixelNormalWS()) {transformSourceType=TRANSFORMSOURCE_World, transformType=TRANSFORM_View}) {g=true}
  n67 = Multiply(A=Multiply(A=Multiply(A=@n64.RGB, B=@n65), B=NamedRerouteUsage()), B=OneMinus(None=Abs(None=@n66)))
  n68 = LinearInterpolate(
   A=ScalarParameter() {defaultValue=0.5, group=06_Thin Film | Side Kick, parameterName=ThinFilmIntensity_Min, sortPriority=2}
   B=ScalarParameter() {defaultValue=0.5, group=06_Thin Film | Side Kick, parameterName=ThinFilmIntensity_Max, sortPriority=2}
   Alpha=ComponentMask(None=NamedRerouteUsage()) {r=true}
  )
  n69 = NamedRerouteUsage()
  n70 = Multiply(A=Multiply(A=Multiply(A=@n67, B=@n68), B=NamedRerouteUsage()), B=Step(X=OneMinus(None=@n69)) {constY=0.9})
  n71 = NamedRerouteUsage()
  n72 = VectorParameter() {group=01_Anime Maps, parameterName=Fresnel_Outline Color, sortPriority=20, defaultValue=<rgba 1 0.939 0.0652 1>}
  n73 = Multiply(A=@n72.RGB, B=ScalarParameter() {defaultValue=1, group=02_Anime Parameters, parameterName=Fresnel_Intensity, sliderMax=10})
  n74 = Add(
   A=Add(A=@n63, B=Multiply(A=@n70, B=@n71))
   B=Multiply(A=NamedRerouteDeclaration(None=@n73) {name=FresnelOutlineColor, nodeColor=<rgba 0 0.4141 1 1>}, B=NamedRerouteUsage())
  )
  n75 = Saturate(None=TwoSidedSign())
  n76 = Multiply(A=LocalPosition().XY, B=ScalarParameter() {defaultValue=15, group=GridEffect, parameterName=LightingWPO_NoiseTiliing})
  n77 = ComponentMask(None=VectorParameter() {parameterName=LightingWPO_NoiseSpeed, defaultValue=<rgba 0.4 0.15 0 1>}.RGB) {g=true, r=true}
  n78 = ScalarParameter() {defaultValue=12, group=GridEffect, parameterName=LightingWPO_JoltStep}
  n79 = Floor(None=Multiply(A=Time(), B=@n78) {constB=12})
  n80 = TextureSampleParameter2D(UVs=Multiply(A=Multiply(A=@n76, B=@n77), B=Divide(A=@n79, B=@n78) {constB=3})) {parameterName=WPO Lightening Noise, texture=&3xPerlin_M}
  n81 = ScalarParameter() {defaultValue=5, group=GridEffect, parameterName=LightingWPO_Spikeness}
  n82 = VectorParameter() {parameterName=LightingWPO_Direction, defaultValue=<rgba 0.4 0.15 0 1>}
  n83 = VectorParameter() {parameterName=LightingWPO_Wave_Direction, defaultValue=<rgba 0.4 0.15 0 1>}
  n84 = LocalPosition() {localOrigin=InstancePreSkinning}
  n85 = PreSkinnedLocalBounds()
  n86 = ComponentMask(None=@n85.Min) {g=true}
  n87 = Divide(A=Subtract(A=ComponentMask(None=@n84.XYZ) {g=true}, B=@n86), B=Subtract(A=ComponentMask(None=@n85.Max) {g=true}, B=@n86))
  n88 = Subtract(
   A=Multiply(A=Saturate(None=@n87), B=ScalarParameter() {defaultValue=2, group=Wave, parameterName=WaveFreq})
   B=Multiply(A=Time(), B=ScalarParameter() {defaultValue=8, group=Wave, parameterName=WaveSpeed})
  )
  n89 = Multiply(A=Add(B=Sine(None=@n88)) {constA=1}, B=ScalarParameter() {defaultValue=0.5, group=Wave, parameterName=FlapAmount})
  n90 = ScalarParameter() {group=Wave, parameterName=Intensity}
  n91 = Add(
   A=Multiply(A=Power(Base=@n80.R, Exp=@n81), B=Transform(None=@n82.RGB) {transformSourceType=TRANSFORMSOURCE_Local})
   B=Multiply(A=Transform(None=@n83.RGB) {transformSourceType=TRANSFORMSOURCE_Local}, B=Multiply(A=@n89, B=@n90))
  )
  n92 = Subtract(
   A=MaterialFunctionCall() {materialFunction=&ObjectPivotPoint}.Object Pivot Location
   B=WorldPosition().XYZ
  )
  n93 = ScalarParameter() {group=GridEffect, parameterName=LightingWPO_Mask Crunch}
  n94 = AppendVector(A=AppendVector(A=@n93, B=@n93), B=Constant())
  n95 = TextureSampleParameter2D() {
   parameterName=WPO Mask
   samplerType=SAMPLERTYPE_Masks
   texture=/BRCosmetics/Weapons/FORT_Melee/Pickaxe_ShovelCropUniversal/Textures/T_ShovelCropUniversal_Pickaxe_Mask
  }
  n96 = StaticSwitchParameter(True=LinearInterpolate(A=@n91, B=Multiply(A=@n92, B=@n94), Alpha=@n95.RGB) {constA=1, constB=0}, False=@n91) {parameterName=Use WPO Mask}
  n97 = Multiply(A=ScalarParameter() {group=GridEffect, parameterName=LightingWPO_Flare}, B=VertexNormalWS()) {constA=10}
  n98 = LocalPosition() {localOrigin=InstancePreSkinning}
  n99 = VectorParameter() {parameterName=LightingWPO_Mask_Pos, defaultValue=<rgba 0.4 0.15 0 1>}
  n100 = ScalarParameter() {defaultValue=200, group=GridEffect, parameterName=LightingWPO_Mask_Radius}
  n101 = ScalarParameter() {defaultValue=1, group=GridEffect, parameterName=LightingWPO_Mask_Hardness}
  SetMaterialAttributes!(
   MaterialAttributes=MaterialFunctionCall!(
    Shading_Gradient=@n56.Shading_Gradient
    Threshold_Gradient=@n56.Shading_Gradient
    Specular_Gradient=@n56.Shading_Gradient
    Sheen_Mask=@n56.Shading_Gradient
    Thickness=@n56.Shading_Gradient
    DFL_Mask=@n56.Shading_Gradient
    FX=@n56.Shading_Gradient
    Metallic_Gradient=@n56.Shading_Gradient
    RimLight_Gradient=@n56.Shading_Gradient
    NormalBuffer=@n56.Shading_Gradient
    --------
    CINE_LightSources=@n56.Shading_Gradient
    CINE_LightColor=@n56.Shading_Gradient
    ---------
    Color_Lit=@n47.Color_Lit
    Color_Shaded=@n47.Color_Lit
    SRM=@n47.Color_Lit
    Emissive=LinearInterpolate(A=@n63, B=@n74, Alpha=@n75)
    DFL_Color=@n47.Color_Lit
    Specular_Color=@n47.Color_Lit
    RimLight_Color=@n47.Color_Lit
   ) {materialFunction=/Game/Global/Materials/Character_Parent_2023/Anime_Parent_2023/MaterialFunctions/MFC_AnimeShader_2023_Post}.Result
   World Position Offset=LinearInterpolate(B=Add(A=@n96, B=@n97), Alpha=SphereMask(A=@n98.XYZ, B=@n99.RGB, Radius=@n100, Hardness=@n101))
  ) {attributeSetTypes [#1]}
  TemporalResponsivenessOutput!(None=LinearInterpolate(A=Constant() {r=0.25}, Alpha=NamedRerouteUsage()) {constB=0.25})
 ]
}
